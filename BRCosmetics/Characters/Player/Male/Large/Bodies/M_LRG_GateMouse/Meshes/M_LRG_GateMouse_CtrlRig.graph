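# Post-process-style Control Rig for the large male GateMouse player body that corrects body deformation and keeps the shoulder pieces in place.
# It chains a shared DeformRig function, weighted parent constraints for shoulder and forearm hide bones, then pose offsets on dyn_shoulder bones.
# Glider-type parachute offsets blend by movement direction through interpolated alphas.
# Left-shoulder fixes derive alphas from swing-twist arm and clavicle rotations, gated by gun, crouch and pistol variables.
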ControlRigBlueprint /BRCosmetics/Characters/Player/Male/Large/Bodies/M_LRG_GateMouse/Meshes/M_LRG_GateMouse_CtrlRig
graphs [
 CollapseNode_ContainedGraph {nodes [
   n1 = RigVMFunctionEntryNode()
   &RigUnit_ParentConstraint(
    AdvancedSettings=(RotationOrderForFilter=XZY)
    Child=(Type=Bone,Name="base_shoulder_01_l")
    Parents=((Item=(Type=Bone,Name="spine_03"),Weight=0.0035),(Item=(Type=Bone,Name="spine_04"),Weight=0.0082),(Item=(Type=Bone,Name="spine_05"),Weight=0.3464),(Item=(Type=Bone,Name="clavicle_l"),Weight=0.642))
    Weight=1
    ExecutePin=@n1.ExecutePin
   )
   > &RigUnit_ParentConstraint(
    AdvancedSettings=(RotationOrderForFilter=XZY)
    Child=(Type=Bone,Name="base_shoulder_02_l")
    Parents=((Item=(Type=Bone,Name="spine_03"),Weight=0.0001),(Item=(Type=Bone,Name="clavicle_r"),Weight=0.0005),(Item=(Type=Bone,Name="spine_04"),Weight=0.0006),(Item=(Type=Bone,Name="spine_05"),Weight=0.2676),(Item=(Type=Bone,Name="clavicle_l"),Weight=0.7313))
    Weight=1
   )
   > &RigUnit_ParentConstraint(
    AdvancedSettings=(RotationOrderForFilter=XZY)
    Child=(Type=Bone,Name="base_shoulder_03_l")
    Parents=((Item=(Type=Bone,Name="spine_03"),Weight=0.0003),(Item=(Type=Bone,Name="clavicle_r"),Weight=0.0017),(Item=(Type=Bone,Name="spine_04"),Weight=0.003),(Item=(Type=Bone,Name="spine_05"),Weight=0.1688),(Item=(Type=Bone,Name="clavicle_l"),Weight=0.8262))
    Weight=1
   )
   > &RigUnit_ParentConstraint(
    AdvancedSettings=(RotationOrderForFilter=XZY)
    Child=(Type=Bone,Name="base_shoulder_01_r")
    Parents=((Item=(Type=Bone,Name="spine_03"),Weight=0.0035),(Item=(Type=Bone,Name="spine_04"),Weight=0.0082),(Item=(Type=Bone,Name="spine_05"),Weight=0.3464),(Item=(Type=Bone,Name="clavicle_r"),Weight=0.642))
    Weight=1
   )
   > &RigUnit_ParentConstraint(
    AdvancedSettings=(RotationOrderForFilter=XZY)
    Child=(Type=Bone,Name="base_shoulder_02_r")
    Parents=((Item=(Type=Bone,Name="spine_03"),Weight=0.0001),(Item=(Type=Bone,Name="clavicle_l"),Weight=0.0005),(Item=(Type=Bone,Name="spine_04"),Weight=0.0006),(Item=(Type=Bone,Name="spine_05"),Weight=0.2676),(Item=(Type=Bone,Name="clavicle_r"),Weight=0.7313))
    Weight=1
   )
   > &RigUnit_ParentConstraint(
    AdvancedSettings=(RotationOrderForFilter=XZY)
    Child=(Type=Bone,Name="base_shoulder_03_r")
    Parents=((Item=(Type=Bone,Name="spine_03"),Weight=0.0003),(Item=(Type=Bone,Name="clavicle_l"),Weight=0.0017),(Item=(Type=Bone,Name="spine_04"),Weight=0.003),(Item=(Type=Bone,Name="spine_05"),Weight=0.1688),(Item=(Type=Bone,Name="clavicle_r"),Weight=0.8262))
    Weight=1
   )
   > &RigUnit_ParentConstraint(
    AdvancedSettings=(InterpolationType=Average,RotationOrderForFilter=XZY)
    Child=(Type=Bone,Name="lowerarm_hideBone_l")
    Parents=((Item=(Type=Bone,Name="upperarm_l"),Weight=1))
    Weight=1
   )
   > n2 = &RigUnit_ParentConstraint(
    AdvancedSettings=(InterpolationType=Average,RotationOrderForFilter=XZY)
    Child=(Type=Bone,Name="lowerarm_hideBone_r")
    Parents=((Item=(Type=Bone,Name="upperarm_r"),Weight=1))
    Weight=1
   )
   RigVMFunctionReturnNode(ExecutePin=@n2.ExecutePin)
  ]}
 CollapseNode_ContainedGraph {nodes [
   n1 = RigVMFunctionEntryNode()
   RigVMFunction_ControlFlowBranch = &RigVMFunction_ControlFlowBranch(ExecuteContext=@n1.ExecutePin, Condition=RigVMVariableNode() {variable=pub_isGliderType}.Value)
   RigVMFunctionReturnNode(ExecutePin=@RigVMFunction_ControlFlowBranch.Completed)
   RigVMFunction_MathBoolToFloat = &RigVMFunction_MathBoolToFloat(Value=RigVMVariableNode() {variable=pub_isParachute}.Value)
   AlphaInterp = &RigVMFunction_AlphaInterp(InterpSpeedDecreasing=5, InterpSpeedIncreasing=5, Scale=1, bClampResult=False, Value=@RigVMFunction_MathBoolToFloat.Result)
   &RigUnit_ModifyTransforms(
    ItemToModify="((Item=(Type=Bone,Name=\"dyn_shoulder_01_l\"),Transform=(Rotation=(X=0,Y=-0.342,Z=0,W=0.9397),Translation=(X=-1,Y=2,Z=-1.2),Scale3D=(X=1,Y=1,Z=1))),(Item=(Type=Bone,Name=\"dyn_shoulder_02_l\"),Transform=(Rotation=(X=0,Y=0,Z=0.2588,W=0.9659),Translation=(X=-4,Y ...(truncated)"
    WeightMinimum=0
    Weight=@AlphaInterp.Result
    ExecutePin=@RigVMFunction_ControlFlowBranch.True
   )
   > &RigUnit_ModifyTransforms(
    ItemToModify="((Item=(Type=Bone,Name=\"dyn_shoulder_01_r\"),Transform=(Rotation=(X=0,Y=0.342,Z=0,W=0.9397),Translation=(X=-1,Y=2,Z=1.2),Scale3D=(X=1,Y=1,Z=1))),(Item=(Type=Bone,Name=\"dyn_shoulder_02_r\"),Transform=(Rotation=(X=0,Y=0,Z=0.2588,W=0.9659),Translation=(X=-5,Y=3 ...(truncated)"
    WeightMinimum=0
    Weight=@AlphaInterp.Result
   )
   Multiply = &RigVMFunction_MathDoubleMul(A=@RigVMFunction_MathBoolToFloat.Result, B=RigVMVariableNode() {variable=Pub_Is_Left}.Value)
   n2 = &RigVMFunction_AlphaInterp(InterpSpeedDecreasing=5, InterpSpeedIncreasing=5, Scale=1, bClampResult=False, Value=@Multiply.Result)
   > &RigUnit_ModifyTransforms(
    ItemToModify="((Item=(Type=Bone,Name=\"dyn_shoulder_01_l\"),Transform=(Rotation=(X=0,Y=0,Z=0,W=1),Translation=(X=0,Y=1,Z=0),Scale3D=(X=1,Y=1,Z=1))),(Item=(Type=Bone,Name=\"dyn_shoulder_02_l\"),Transform=(Rotation=(X=0,Y=0,Z=0,W=1),Translation=(X=-1,Y=2,Z=0),Scale3D=(X=1,Y=1 ...(truncated)"
    WeightMinimum=0
    Weight=@n2.Result
   )
   n3 = &RigVMFunction_AlphaInterp(
    InterpSpeedDecreasing=5
    InterpSpeedIncreasing=5
    Scale=1
    bClampResult=False
    Value=&RigVMFunction_MathDoubleMul(A=@RigVMFunction_MathBoolToFloat.Result, B=RigVMVariableNode() {variable=Pub_Is_Right}.Value).Result
   )
   > &RigUnit_ModifyTransforms(
    ItemToModify="((Item=(Type=Bone,Name=\"dyn_shoulder_01_r\"),Transform=(Rotation=(X=0,Y=0,Z=0,W=1),Translation=(X=0,Y=2,Z=0),Scale3D=(X=1,Y=1,Z=1))),(Item=(Type=Bone,Name=\"dyn_shoulder_02_r\"),Transform=(Rotation=(X=0,Y=0,Z=0,W=1),Translation=(X=2,Y=2,Z=0),Scale3D=(X=1,Y=1, ...(truncated)"
    WeightMinimum=0
    Weight=@n3.Result
   )
   n4 = &RigVMFunction_AlphaInterp(
    InterpSpeedDecreasing=5
    InterpSpeedIncreasing=5
    Scale=1
    bClampResult=False
    Value=&RigVMFunction_MathDoubleMul(A=@RigVMFunction_MathBoolToFloat.Result, B=RigVMVariableNode() {variable=Pub_Is_Forward}.Value).Result
   )
   > &RigUnit_ModifyTransforms(
    ItemToModify="((Item=(Type=Bone,Name=\"dyn_shoulder_01_l\"),Transform=(Rotation=(X=0,Y=0,Z=0,W=1),Translation=(X=0,Y=0,Z=0),Scale3D=(X=1,Y=1,Z=1))),(Item=(Type=Bone,Name=\"dyn_shoulder_02_l\"),Transform=(Rotation=(X=0,Y=0,Z=-0.0872,W=0.9962),Translation=(X=1,Y=0,Z=0),Scale3 ...(truncated)"
    WeightMinimum=0
    Weight=@n4.Result
   )
   > &RigUnit_ModifyTransforms(
    ItemToModify="((Item=(Type=Bone,Name=\"dyn_shoulder_01_r\"),Transform=(Rotation=(X=0,Y=0,Z=0,W=1),Translation=(X=0,Y=0,Z=0),Scale3D=(X=1,Y=1,Z=1))),(Item=(Type=Bone,Name=\"dyn_shoulder_02_r\"),Transform=(Rotation=(X=0,Y=0,Z=-0.1736,W=0.9848),Translation=(X=1,Y=0,Z=0),Scale3 ...(truncated)"
    WeightMinimum=0
    Weight=@n4.Result
   )
  ]}
 CollapseNode_ContainedGraph {nodes [
   RigVMFunctionEntryNode()
   > GetAlphaFromQuatDelta = RigVMFunctionReferenceNode(
    Rotation Order=ZYX
    child=(Type=Bone,Name="upperarm_l")
    debugRotation=false
    outputMinMax=(X=0,Y=1)
    parent=(Type=Null,Name="upperarm_l_null")
    premult=1
    rotationMinMax=(X=0,Y=-20)
    triggerAxis=Y
    twistAxis=(X=1,Y=0,Z=0)
   )
   n1 = &RigVMFunction_AlphaInterp(
    InterpSpeedDecreasing=5
    InterpSpeedIncreasing=5
    Scale=1
    bClampResult=False
    Value=&RigVMFunction_MathDoubleMul(A=@GetAlphaFromQuatDelta.alpha, B=RigVMVariableNode() {variable=Pub_IsGunOut}.Value).Result
   )
   > &RigUnit_ModifyTransforms(
    ItemToModify="((Item=(Type=Bone,Name=\"dyn_shoulder_02_l\"),Transform=(Rotation=(X=0,Y=0,Z=0.1736,W=0.9848),Translation=(X=0,Y=2,Z=0),Scale3D=(X=1,Y=1,Z=1))),(Item=(Type=Bone,Name=\"dyn_shoulder_01_l\"),Transform=(Rotation=(X=0.0449,Y=-0.1677,Z=0.2549,W=0.9513),Translation= ...(truncated)"
    WeightMinimum=0
    Weight=@n1.Result
   )
   n2 = RigVMVariableNode() {variable=pub_isCrouching}
   > &RigUnit_ModifyTransforms(
    ItemToModify="((Item=(Type=Bone,Name=\"dyn_shoulder_01_l\"),Transform=(Rotation=(X=-0.2094,Y=0.1209,Z=0.2801,W=0.929),Translation=(X=-5,Y=2,Z=-3),Scale3D=(X=1,Y=1,Z=1))),(Item=(Type=Bone,Name=\"dyn_shoulder_02_l\"),Transform=(Rotation=(X=0,Y=0,Z=0.1736,W=0.9848),Translation ...(truncated)"
    WeightMinimum=0
    Weight=&RigVMFunction_MathDoubleMul(A=@n1.Result, B=&RigVMFunction_MathBoolToFloat(Value=@n2.Value).Result).Result
   )
   > GetAlphaFromQuatDelta_1 = RigVMFunctionReferenceNode(
    Rotation Order=ZYX
    child=(Type=Bone,Name="clavicle_l")
    debugRotation=false
    outputMinMax=(X=0,Y=1)
    parent=(Type=Null,Name="clavicle_l_null")
    premult=1
    rotationMinMax=(X=0,Y=-45)
    triggerAxis=Y
    twistAxis=(X=1,Y=0,Z=0)
   )
   Multiply_5 = RigVMAggregateNode(
    A=@GetAlphaFromQuatDelta_1.alpha
    C=RigVMDispatchNode(False=1, True=0, Condition=RigVMVariableNode() {variable=pub_isParachute}.Value) {template="DISPATCH_RigVMDispatch_If(in Condition,in True,in False,out Result)"}.Result
    B=RigVMVariableNode() {variable=Pub_IsPistol}.Value
   )
   > &RigUnit_ModifyTransforms(
    ItemToModify="((Item=(Type=Bone,Name=\"dyn_shoulder_02_l\"),Transform=(Rotation=(X=0,Y=0,Z=0.1736,W=0.9848),Translation=(X=-4,Y=3,Z=-2),Scale3D=(X=1,Y=1,Z=1))),(Item=(Type=Bone,Name=\"dyn_shoulder_01_l\"),Transform=(Rotation=(X=0.067,Y=-0.25,Z=0.25,W=0.933),Translation=(X=- ...(truncated)"
    WeightMinimum=0
    Weight=@Multiply_5.Result
   )
   > RigVMFunctionReturnNode()
  ]}
 Multiply_1_ContainedGraph {nodes [
   n1 = RigVMFunctionEntryNode()
   RigVMFunctionReturnNode(Result=&RigVMFunction_MathDoubleMul(B=@n1.C, A=&RigVMFunction_MathDoubleMul(A=@n1.A, B=@n1.B).Result).Result)
  ]}
 New Function_ContainedGraph {nodes [
   RigVMFunctionEntryNode()
   RigVMFunction_MathBoolToInteger = &RigVMFunction_MathBoolToInteger(Value=RigVMVariableNode() {variable=onlyTwist}.Value)
   RigVMFunction_MathQuaternionSwingTwist = &RigVMFunction_MathQuaternionSwingTwist(
    Input=&RigUnit_GetRelativeTransformForItem(
     bChildInitial=False
     bParentInitial=False
     Child=RigVMVariableNode() {variable=child}.Value
     Parent=RigVMVariableNode() {variable=parent}.Value
    ).RelativeTransform.Rotation
    TwistAxis=RigVMVariableNode() {variable=twistAxis}.Value
   )
   n1 = RigVMDispatchNode(
    Values=((),())
    Index=@RigVMFunction_MathBoolToInteger.Result
    Values.0=@RigVMFunction_MathQuaternionSwingTwist.Swing
    Values.1=@RigVMFunction_MathQuaternionSwingTwist.Twist
   ) {template="DISPATCH_RigVMDispatch_SelectInt32(in Index,in Values,out Result)"}
   > RigVMVariableNode(Value=@n1.Result) {variable=entry_outputQuat}
   RigVMFunction_MathQuaternionToEuler = &RigVMFunction_MathQuaternionToEuler(Value=@n1.Result, RotationOrder=RigVMVariableNode() {variable=Rotation Order}.Value)
   > DISPATCH_RigVMDispatch_Print = RigVMDispatchNode(
    ScreenColor=(R=1,G=1,B=1,A=1)
    ScreenDuration=1.5
    Value=@RigVMFunction_MathQuaternionToEuler.Result
    Enabled=RigVMVariableNode() {variable=debugRotation}.Value
   ) {template="DISPATCH_RigVMDispatch_Print(in Prefix,in Value,in Enabled,in ScreenDuration,in ScreenColor)"}
   Cast = RigVMDispatchNode(Value=RigVMVariableNode() {variable=triggerAxis}.Value) {template="DISPATCH_RigVMDispatch_CastEnumToInt(in Value,out Result)"}
   DISPATCH_RigVMDispatch_SelectInt32 = RigVMDispatchNode(
    Values=(0,0,0)
    Index=@Cast.Result
    Values.0=@RigVMFunction_MathQuaternionToEuler.Result.X
    Values.1=@RigVMFunction_MathQuaternionToEuler.Result.Y
    Values.2=@RigVMFunction_MathQuaternionToEuler.Result.Z
   ) {template="DISPATCH_RigVMDispatch_SelectInt32(in Index,in Values,out Result)"}
   n2 = RigVMVariableNode() {variable=rotationMinMax}
   n3 = RigVMVariableNode() {variable=outputMinMax}
   Remap = &RigVMFunction_MathDoubleRemap(
    bClamp=False
    Value=@DISPATCH_RigVMDispatch_SelectInt32.Result
    SourceMinimum=@n2.Value.X
    SourceMaximum=@n2.Value.Y
    TargetMinimum=@n3.Value.X
    TargetMaximum=@n3.Value.Y
   )
   Absolute = &RigVMFunction_MathDoubleAbs(Value=@Remap.Result)
   n4 = RigVMVariableNode() {variable=absoluteValue}
   Multiply = &RigVMFunction_MathDoubleMul(
    A=&RigVMFunction_MathDoubleClamp(Value=RigVMDispatchNode(True=@Absolute.Result, False=@Remap.Result, Condition=@n4.Value) {template="DISPATCH_RigVMDispatch_If(in Condition,in True,in False,out Result)"}.Result).Result
    B=RigVMVariableNode() {variable=premult}.Value
   )
   > RigVMFunctionReturnNode(alpha=@Multiply.Result, rawQuat=RigVMVariableNode() {variable=entry_outputQuat}.Value)
  ]}
 RigVMFunctionLibrary {nodes [GetAlphaFromQuatDelta = RigVMCollapseNode(
    ExecuteContext=()
    Rotation Order=ZYX
    child=(Type=Bone,Name="None")
    outputMinMax=(X=0,Y=1)
    parent=(Type=Bone,Name="None")
    premult=1
    rotationMinMax=(X=0,Y=0)
    twistAxis=(X=1,Y=0,Z=0)
   )]}
 RigVMModel {nodes [
   RigUnit_BeginExecution = &RigUnit_BeginExecution()
   FN_DeformRig_FNC = RigVMFunctionReferenceNode(
    DeformRig_ Curve="(deformRig_l_upperarm_up_alpha_72_#1=true,deformRig_r_upperarm_up_alpha_73_#2=true,deformRig_l_upperarm_fr_alpha_74_#3=true,deformRig_r_upperarm_fr_alpha_75_CB51B80C4 ...(truncated)"
    DeformRig_Bool="(upperarm_1_#4=true,elbow_3_#5=true,wrist_7_#6=True,groin_8_#7=true,glute_17_#8=true,knee_9_40C408324CA261 ...(truncated)"
    DeformRig_Debug=(upperarm_debug_18_#4=False,wrist_debug_20_#6=False,lat_debug_26_#9=False,groinGlute_debug_27_#7=False)
    DeformRig_Value="(upperarm_up_18_#10=(Rotation=(X=0,Y=0,Z=0,W=1),Translation=(X=0,Y=0,Z=0.8),Scale3D=(X=0.7,Y=1,Z=1)),upperarm_fwd_fr_19_#11=(Rotation=(X=0,Y=0,Z=0,W=1),Translation=(X=0,Y=10,Z=0),Scale3D=(X=0.5,Y=1, ...(truncated)"
    ExecuteContext=@RigUnit_BeginExecution.ExecutePin
   )
   Constraints = RigVMCollapseNode(ExecutePin=@FN_DeformRig_FNC.ExecuteContext)
   OpenedParachute = RigVMCollapseNode(ExecutePin=@Constraints.ExecutePin)
   ExtraFixes = RigVMCollapseNode(ExecuteContext=@OpenedParachute.ExecutePin)
  ]}
]
hierarchy [
 root {loc=<0 0 0>, rot=<0 0 0 1>}
 pelvis {parent=root, loc=<0 -0.7321 92.4218>, rot=<0.0127 0.707 -0.0127 -0.707>}
 spine_01 {parent=pelvis, loc=<5.6002 -0.5128 0>, rot=<0 0 0.034 -0.9994>}
 spine_02 {parent=spine_01, loc=<8.0756 0.0139 0>, rot=<0 0 0.0238 -0.9997>}
 spine_03 {parent=spine_02, loc=<7.0395 0 0>, rot=<0 0 0.0149 -0.9999>}
 spine_04 {parent=spine_03, loc=<8.1023 0 0>, rot=<0 0 -0.0132 -0.9999>}
 spine_05 {parent=spine_04, loc=<10.0141 0 0>, rot=<0 0 -0.0015 -1>}
 clavicle_l {parent=spine_05, loc=<2.5856 -0.2866 -4.3777>, rot=<0.0465 -0.7289 -0.009 -0.6829>}
 upperarm_l {parent=clavicle_l, loc=<18.2264 0.1342 -0.001>, rot=<0.0253 -0.2693 0.0013 -0.9627>}
 lowerarm_l {parent=upperarm_l, loc=<29.0071 0.1686 -1.239>, rot=<0.0391 -0.0099 -0.2453 -0.9686>}
 hand_l {parent=lowerarm_l, loc=<25.8607 0.222 1.0274>, rot=<-0.7193 -0.0988 0.1193 -0.6773>}
 thumb_01_l {parent=hand_l, loc=<1.3118 -0.3191 -2.459>, rot=<0.6546 -0.4042 0.0664 -0.6354>}
 thumb_02_l {parent=thumb_01_l, loc=<5.9336 -0.0448 0.0401>, rot=<-0.0037 0.0213 0.298 -0.9543>}
 thumb_03_l {parent=thumb_02_l, loc=<2.2223 0.3682 -0.0907>, rot=<0.0118 -0.013 0.0066 -0.9998>}
 index_metacarpal_l {parent=hand_l, loc=<1.6535 0.0183 -2.96>, rot=<0.0158 -0.0124 -0.0173 -0.9996>}
 index_01_l {parent=index_metacarpal_l, loc=<8.3008 0.0357 -0.0524>, rot=<-0.0044 0.0485 0.0924 -0.9945>}
 index_02_l {parent=index_01_l, loc=<3.3587 -0.3004 0.0452>, rot=<-0.002 0.0027 0.3225 -0.9466>}
 index_03_l {parent=index_02_l, loc=<2.5578 -0.123 -0.0252>, rot=<-0.002 0.0067 0.2558 -0.9667>}
 middle_metacarpal_l {parent=hand_l, loc=<2.1856 -0.1725 -0.5984>, rot=<0.006 0.0168 -0.0329 -0.9993>}
 middle_01_l {parent=middle_metacarpal_l, loc=<7.3275 -0.2447 0.0129>, rot=<-0.0208 0.0042 0.0858 -0.9961>}
 middle_02_l {parent=middle_01_l, loc=<4.0123 0.1401 0.1175>, rot=<-0.0218 0.0258 0.3512 -0.9357>}
 middle_03_l {parent=middle_02_l, loc=<2.6077 0.0534 0.1373>, rot=<0.0082 0.0188 0.1708 -0.9851>}
 ring_metacarpal_l {parent=hand_l, loc=<2.2074 -0.3625 1.3638>, rot=<-0.0284 0.0722 -0.0458 -0.9959>}
 ring_01_l {parent=ring_metacarpal_l, loc=<7.0295 -0.6197 -0.1446>, rot=<-0.0489 -0.0252 0.1892 -0.9804>}
 ring_02_l {parent=ring_01_l, loc=<3.6174 0.4647 -0.1027>, rot=<0 0 0.2806 -0.9598>}
 ring_03_l {parent=ring_02_l, loc=<2.4152 0.114 0.0199>, rot=<0 0 0.1189 -0.9929>}
 pinky_metacarpal_l {parent=hand_l, loc=<1.7616 -0.9753 3.3814>, rot=<-0.0397 0.0814 -0.069 -0.9935>}
 pinky_01_l {parent=pinky_metacarpal_l, loc=<7.1709 -0.4831 -0.1094>, rot=<-0.0654 -0.0035 0.2242 -0.9723>}
 pinky_02_l {parent=pinky_01_l, loc=<2.8274 0.0893 -0.1043>, rot=<0.0037 0.0011 0.2854 -0.9584>}
 pinky_03_l {parent=pinky_02_l, loc=<2.2046 0.0307 -0.0078>, rot=<-0.0002 0.0098 0.1445 -0.9895>}
 lowerarm_twist_02_l {parent=lowerarm_l, loc=<9.6199 0 0>, rot=<0 0 0 -1>}
 lowerarm_twist_01_l {parent=lowerarm_l, loc=<17.8549 0 0>, rot=<0 0 0 -1>}
 lowerarm_hideBone_l {parent=lowerarm_l, loc=<2.2038 7.4377 1.5326>, rot=<-0.0089 0.0019 0.4164 -0.9092>}
 dyn_lowerarm_hideBone_l {parent=lowerarm_l, loc=<0 0 0>, rot=<0 0 0 1>}
 upperarm_twist_01_l {parent=upperarm_l, loc=<7.9796 0.0843 -0.6195>, rot=<0 0 0 -1>}
 upperarm_twist_02_l {parent=upperarm_l, loc=<19.5059 0.0843 -0.6195>, rot=<0 0 0 -1>}
 Backpack_Custom {parent=upperarm_l, type=Null, loc=<0 0 0>, rot=<0 0 0 1>}
 base_shoulder_01_l {parent=clavicle_l, loc=<13.3661 8.4946 8.4853>, rot=<-0.3928 -0.586 0.5694 -0.4219>}
 dyn_shoulder_01_l {parent=base_shoulder_01_l, loc=<0 0 0>, rot=<0 0 0 1>}
 base_shoulder_02_l {parent=clavicle_l, loc=<15.7063 -2.3042 13.1043>, rot=<-0.3928 -0.586 0.5694 -0.4219>}
 dyn_shoulder_02_l {parent=base_shoulder_02_l, loc=<0 0 0>, rot=<0 0 0 1>}
 base_shoulder_03_l {parent=clavicle_l, loc=<13.2693 -11.2862 5.1789>, rot=<-0.3928 -0.586 0.5694 -0.4219>}
 dyn_shoulder_03_l {parent=base_shoulder_03_l, loc=<0 0 0>, rot=<0 0 0 1>}
 upperarm_l_null {parent=clavicle_l, type=Null, loc=<18.2264 0.1342 -0.001>, rot=<0.0253 -0.2693 0.0013 -0.9627>}
 neck_01 {parent=spine_05, loc=<10.815 1.9196 0>, rot=<0 0 -0.1864 -0.9825>}
 neck_02 {parent=neck_01, loc=<4.7347 0.3045 0>, rot=<0 0 0.0363 -0.9993>}
 head {parent=neck_02, loc=<5.7632 0.5552 0>, rot=<0 0 0.115 -0.9934>}
 clavicle_r {parent=spine_05, loc=<2.5855 -0.2866 4.3777>, rot=<0.7289 0.0465 -0.6829 0.009>}
 upperarm_r {parent=clavicle_r, loc=<-18.2264 -0.1342 0.001>, rot=<0.0253 -0.2693 0.0013 -0.9627>}
 lowerarm_r {parent=upperarm_r, loc=<-29.0071 -0.1686 1.239>, rot=<0.0391 -0.0099 -0.2453 -0.9686>}
 hand_r {parent=lowerarm_r, loc=<-25.8607 -0.222 -1.0274>, rot=<-0.7193 -0.0988 0.1193 -0.6773>}
 thumb_01_r {parent=hand_r, loc=<-1.3118 0.3191 2.459>, rot=<0.6546 -0.4042 0.0664 -0.6354>}
 thumb_02_r {parent=thumb_01_r, loc=<-5.9336 0.0448 -0.0401>, rot=<-0.0037 0.0213 0.298 -0.9543>}
 thumb_03_r {parent=thumb_02_r, loc=<-2.2223 -0.3682 0.0907>, rot=<0.0118 -0.013 0.0066 -0.9998>}
 index_metacarpal_r {parent=hand_r, loc=<-1.6535 -0.0183 2.96>, rot=<0.0158 -0.0124 -0.0173 -0.9996>}
 index_01_r {parent=index_metacarpal_r, loc=<-8.3008 -0.0357 0.0524>, rot=<-0.0044 0.0485 0.0924 -0.9945>}
 index_02_r {parent=index_01_r, loc=<-3.3587 0.3004 -0.0452>, rot=<-0.002 0.0027 0.3225 -0.9466>}
 index_03_r {parent=index_02_r, loc=<-2.5578 0.123 0.0252>, rot=<-0.002 0.0067 0.2558 -0.9667>}
 middle_metacarpal_r {parent=hand_r, loc=<-2.1856 0.1725 0.5984>, rot=<0.006 0.0168 -0.0329 -0.9993>}
 middle_01_r {parent=middle_metacarpal_r, loc=<-7.3275 0.2447 -0.0129>, rot=<-0.0208 0.0042 0.0858 -0.9961>}
 middle_02_r {parent=middle_01_r, loc=<-4.0123 -0.1401 -0.1175>, rot=<-0.0218 0.0258 0.3512 -0.9357>}
 middle_03_r {parent=middle_02_r, loc=<-2.6077 -0.0534 -0.1373>, rot=<0.0082 0.0188 0.1708 -0.9851>}
 ring_metacarpal_r {parent=hand_r, loc=<-2.2074 0.3625 -1.3638>, rot=<-0.0284 0.0722 -0.0458 -0.9959>}
 ring_01_r {parent=ring_metacarpal_r, loc=<-7.0295 0.6197 0.1446>, rot=<-0.0489 -0.0252 0.1892 -0.9804>}
 ring_02_r {parent=ring_01_r, loc=<-3.6174 -0.4647 0.1027>, rot=<0 0 0.2806 -0.9598>}
 ring_03_r {parent=ring_02_r, loc=<-2.4152 -0.114 -0.0199>, rot=<0 0 0.1189 -0.9929>}
 pinky_metacarpal_r {parent=hand_r, loc=<-1.7616 0.9753 -3.3814>, rot=<-0.0397 0.0814 -0.069 -0.9935>}
 pinky_01_r {parent=pinky_metacarpal_r, loc=<-7.1709 0.4831 0.1094>, rot=<-0.0654 -0.0035 0.2242 -0.9723>}
 pinky_02_r {parent=pinky_01_r, loc=<-2.8274 -0.0893 0.1043>, rot=<0.0037 0.0011 0.2854 -0.9584>}
 pinky_03_r {parent=pinky_02_r, loc=<-2.2046 -0.0307 0.0078>, rot=<-0.0002 0.0098 0.1445 -0.9895>}
 lowerarm_twist_02_r {parent=lowerarm_r, loc=<-9.6199 0 0>, rot=<0 0 0 -1>}
 lowerarm_twist_01_r {parent=lowerarm_r, loc=<-17.8549 0 0>, rot=<0 0 0 -1>}
 lowerarm_hideBone_r {parent=lowerarm_r, loc=<-2.2038 -7.4377 -1.5326>, rot=<0 0 0.3775 -0.926>}
 dyn_lowerarm_hideBone_r {parent=lowerarm_r, loc=<0 0 0>, rot=<0 0 0 1>}
 upperarm_twist_01_r {parent=upperarm_r, loc=<-7.9796 -0.0843 0.6195>, rot=<0 0 0 -1>}
 upperarm_twist_02_r {parent=upperarm_r, loc=<-19.5059 -0.0843 0.6195>, rot=<0 0 0 -1>}
 base_shoulder_01_r {parent=clavicle_r, loc=<-13.3661 -8.4946 -8.4855>, rot=<-0.586 0.3928 -0.4219 -0.5694>}
 dyn_shoulder_01_r {parent=base_shoulder_01_r, loc=<0 0 0>, rot=<0 0 0 1>}
 base_shoulder_02_r {parent=clavicle_r, loc=<-15.7063 2.3042 -13.1039>, rot=<-0.586 0.3928 -0.4219 -0.5694>}
 dyn_shoulder_02_r {parent=base_shoulder_02_r, loc=<0 0 0>, rot=<0 0 0 1>}
 base_shoulder_03_r {parent=clavicle_r, loc=<-13.2694 11.2862 -5.1786>, rot=<-0.586 0.3928 -0.4219 -0.5694>}
 dyn_shoulder_03_r {parent=base_shoulder_03_r, loc=<0 0 0>, rot=<0 0 0 1>}
 upperarm_r_null {parent=clavicle_r, type=Null, loc=<-18.2264 -0.1342 0.001>, rot=<0.0253 -0.2693 0.0013 -0.9627>}
 attach_cape {parent=spine_05, loc=<9.017 -11.714 0>, rot=<0.0616 -0.7044 -0.0616 -0.7044>}
 cape {parent=attach_cape, type=Null, loc=<0 -8 0>, rot=<0 0 0 1>}
 attach_backpack {parent=spine_05, loc=<-0.123 -13.473 0>, rot=<-0.0093 -0.707 0.0093 -0.707>}
 Backpack_BR {parent=attach_backpack, type=Null, loc=<0 -3 0>, rot=<0 0 0 1>}
 head_null {parent=spine_05, type=Null, loc=<20.4455 6.1834 0>, rot=<0 0 -0.036 -0.9994>}
 clavicle_l_null {parent=spine_05, type=Null, loc=<2.5856 -0.2866 -4.3777>, rot=<0.0465 -0.7289 -0.009 -0.6829>}
 clavicle_r_null {parent=spine_05, type=Null, loc=<2.5855 -0.2866 4.3777>, rot=<0.7289 0.0465 -0.6829 0.009>}
 thigh_l {parent=pelvis, loc=<-1.8716 -0.0109 -11.9936>, rot=<0.0012 0.0572 0.0155 -0.9982>}
 calf_l {parent=thigh_l, loc=<-41.9178 -0.2246 0.0926>, rot=<0.0006 -0.0161 -0.0391 -0.9991>}
 foot_l {parent=calf_l, loc=<-38.7057 1.2438 0.2452>, rot=<0.0975 -0.0474 0.0317 -0.9936>}
 ball_l {parent=foot_l, loc=<-6.1614 13.1469 0.6807>, rot=<0.0001 -0.0001 -0.7063 -0.7079>}
 calf_twist_02_l {parent=calf_l, loc=<-15.9268 -0.0967 0.0397>, rot=<0.0237 0.0095 -0.0035 -0.9997>}
 calf_twist_01_l {parent=calf_l, loc=<-32.9335 -0.0967 0.0397>, rot=<0.0237 0.0095 -0.0035 -0.9997>}
 deform_knee_out_base_l {parent=calf_l, loc=<0 6.8739 0>, rot=<0 0 0 -1>}
 deform_knee_out_l {parent=deform_knee_out_base_l, loc=<0 0 0>, rot=<0 0 0 1>}
 thigh_twist_01_l {parent=thigh_l, loc=<-21.6303 0 0>, rot=<0.0013 0 0 -1>}
 thigh_r {parent=pelvis, loc=<-1.8716 -0.0109 11.9936>, rot=<0.0572 -0.0012 0.9982 0.0155>}
 calf_r {parent=thigh_r, loc=<41.9178 0.2246 -0.0926>, rot=<0.0006 -0.0161 -0.0391 -0.9991>}
 foot_r {parent=calf_r, loc=<38.7057 -1.2438 -0.2452>, rot=<0.0975 -0.0474 0.0317 -0.9936>}
 ball_r {parent=foot_r, loc=<6.1614 -13.1469 -0.6807>, rot=<0.0001 -0.0001 -0.7063 -0.7079>}
 calf_twist_02_r {parent=calf_r, loc=<15.9268 0.0967 -0.0397>, rot=<0.0237 0.0095 -0.0035 -0.9997>}
 calf_twist_01_r {parent=calf_r, loc=<32.9335 0.0967 -0.0397>, rot=<0.0237 0.0095 -0.0035 -0.9997>}
 deform_knee_out_base_r {parent=calf_r, loc=<0 -6.8739 0>, rot=<0 0 0 -1>}
 deform_knee_out_r {parent=deform_knee_out_base_r, loc=<0 0 0>, rot=<0 0 0 1>}
 thigh_twist_01_r {parent=thigh_r, loc=<21.6303 0 0>, rot=<0.0013 0 0 -1>}
 thigh_l_null {parent=pelvis, type=Null, loc=<-1.8716 -0.0109 -11.9936>, rot=<0.0012 0.0572 0.0155 -0.9982>}
 thigh_r_null {parent=pelvis, type=Null, loc=<-1.8716 -0.0109 11.9936>, rot=<0.0572 -0.0012 0.9982 0.0155>}
]
variables [
 Pub_IsGunOut {type=double, default=0}
 Pub_IsPistol {type=double, default=0}
 Pub_IsTargeting {type=bool, default=False}
 Pub_Is_Backward {type=double, default=0}
 Pub_Is_Forward {type=double, default=0}
 Pub_Is_Left {type=double, default=0}
 Pub_Is_Right {type=double, default=0}
 pub_isCrouching {type=bool, default=False}
 pub_isGliderType {type=bool, default=False}
 pub_isParachute {type=bool, default=False}
]
