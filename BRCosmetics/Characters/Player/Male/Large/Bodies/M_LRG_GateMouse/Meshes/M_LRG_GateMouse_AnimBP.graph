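# Animation Blueprint for the M_LRG_GateMouse player body cosmetic, layering Control Rig animation over the pose copied from its attached parent mesh.
# Events derive clamped directional movement, gun/pistol and glider values, feeding the rig alongside crouch, targeting and parachute state.
# After the rig, two passes of additive bone-space scaling adjust both lower-arm hide bones.
# Two pickaxe meta-tag flags update each frame but go unread here.
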AnimBlueprint /BRCosmetics/Characters/Player/Male/Large/Bodies/M_LRG_GateMouse/Meshes/M_LRG_GateMouse_AnimBP
target_skeleton=&/M_LRG_GateMouse_Skeleton
graphs [
 AnimGraph {nodes [
   n1 = Control Rig!(
    Source=Copy Pose From Mesh() {bUseAttachedParent=true}
    Alpha
    pub_isParachute:bool=Get bIsParachuteOpened()
    Pub_Is_Left:double=Get strafeLeft float() -> (strafeLeft float:double)
    Pub_Is_Right:double=Get strafeRight float() -> (strafeRight float:double)
    Pub_Is_Forward:double=Get moveForward() -> (moveForward:double)
    Pub_Is_Backward:double=Get moveBackward() -> (moveBackward:double)
    Pub_IsGunOut:double=Get IsGunOut() -> (IsGunOut:double)
    pub_isCrouching:bool=Get bIsCrouching()
    Pub_IsPistol:double=Get IsPistolOut() -> (IsPistolOut:double)
    Pub_IsTargeting:bool=Get bIsTargeting()
    pub_isGliderType:bool=Get HangGliding() -> (HangGliding:bool)
   ) -> (Pose) {controlRigAssetReference {blueprintRigClass=&/M_LRG_GateMouse_CtrlRig.M_LRG_GateMouse_CtrlRig_C}}
   n2 = "Transform (Modify) Bone - Bone: dyn_lowerarm_hideBone_l"!(ComponentPose=Local To Component(LocalPose=@n1), bAlphaBoolEnabled:bool=True, Translation="0,0,0", Rotation="0,0,0", Scale="0.85,0.85,0.85") -> (Pose) {
    alphaInputType=Bool
    scaleMode=BMM_Additive
    scaleSpace=BCS_BoneSpace
    boneToModify {boneName=dyn_lowerarm_hideBone_l}
    scale=<0.85 0.85 0.85>
   }
   n3 = "Transform (Modify) Bone - Bone: dyn_lowerarm_hideBone_r"!(ComponentPose=@n2, bAlphaBoolEnabled:bool=True, Translation="0,0,0", Rotation="0,0,0", Scale="0.85,0.85,0.85") -> (Pose) {
    alphaInputType=Bool
    scaleMode=BMM_Additive
    scaleSpace=BCS_BoneSpace
    boneToModify {boneName=dyn_lowerarm_hideBone_r}
    scale=<0.85 0.85 0.85>
   }
   n4 = "Transform (Modify) Bone - Bone: dyn_lowerarm_hideBone_l"(ComponentPose=@n3, bAlphaBoolEnabled=True, Translation="0,0,0", Rotation="0,0,0", Scale="0.9,0.9,0.9") {alphaInputType=Bool, scaleMode=BMM_Additive, scaleSpace=BCS_BoneSpace, boneToModify {boneName=dyn_lowerarm_hideBone_l}, scale=<0.9 0.9 0.9>}
   n5 = "Transform (Modify) Bone - Bone: dyn_lowerarm_hideBone_r"(ComponentPose=@n4, bAlphaBoolEnabled=True, Translation="0,0,0", Rotation="0,0,0", Scale="0.9,0.9,0.9") {alphaInputType=Bool, scaleMode=BMM_Additive, scaleSpace=BCS_BoneSpace, boneToModify {boneName=dyn_lowerarm_hideBone_r}, scale=<0.9 0.9 0.9>}
   Output Pose(Result=Component To Local(ComponentPose=@n5)) {layerGroup=None}
  ]}
 EventGraph {nodes [
   Event BlueprintUpdateAnimation()
   > Sequence()
   > n1 = PlayerMovement(AnimInstance=Self-Reference(), XDotProductCoefficient=1, YDotProductCoefficient=1, ZDotProductCoefficient=1)
   n2 = BreakVector(InVec=MakeVector(X=@n1, Y=@n1.YMovement))
   > Set strafeLeft float(strafeLeft float:double="Absolute (Float)"(A="Clamp (Float)"(Value=@n2, Min=-1, Max=0))) -> (Output_Get:double)
   > Set strafeRight float(strafeRight float:double="Clamp (Float)"(Value=@n2, Min=0, Max=1)) -> (Output_Get:double)
   > Set moveForward(moveForward:double="Clamp (Float)"(Value=@n2.Y, Min=0, Max=1)) -> (Output_Get:double)
   > Set moveBackward(moveBackward:double="Absolute (Float)"(A="Clamp (Float)"(Value=@n2.Y, Min=-1, Max=0))) -> (Output_Get:double)
   > Set IswarmshadeweaselPA(IswarmshadeweaselPA:bool=PickaxeItemDefinitionHasAnyMetaTag(Tags=Make Array("[0]"=(TagName="Cosmetics.MetaTags.WarmShadeWeasel")), bRequiresEquipped=true)) -> (Output_Get:bool)
   > Set IsSeaSaltlPA(IsSeaSaltlPA:bool=PickaxeItemDefinitionHasAnyMetaTag(Tags=Make Array("[0]"=(TagName="Cosmetics.Metatags.SeaSalt")), bRequiresEquipped=true)) -> (Output_Get:bool)
   Event OnParachuteOpened()
   > Switch on EGliderType(Selection=Get PlayerGliderType() -> (PlayerGliderType=HangGlider))
   > Set HangGliding(HangGliding:bool=true) -> (Output_Get:bool)
   Event OnParachuteClosed()
   > Set HangGliding(HangGliding=false)
   Event OnWeaponChanged()
   > Set IsGunOut(IsGunOut:double=0) -> (Output_Get:double)
   > Set IsPistolOut(IsPistolOut:double=0) -> (Output_Get:double)
   > n3 = Switch on EFortWeaponCoreAnimation!(execute, Selection:EFortWeaponCoreAnimation Enum=Get WeaponCoreAnim!() -> (WeaponCoreAnim)) -> (
    Melee:Exec
    Pistol:Exec
    Shotgun:Exec
    PaperBlueprint:Exec
    Rifle:Exec
    MeleeOneHand:Exec
    MachinePistol:Exec
    RocketLauncher:Exec
    GrenadeLauncher:Exec
    GoingCommando:Exec
    AssaultRifle:Exec
    TacticalShotgun:Exec
    SniperRifle:Exec
    TrapPlacement:Exec
    ShoulderLauncher:Exec
    AbilityDecoTool:Exec
    Crossbow:Exec
    C4:Exec
    RemoteControl:Exec
    DualWield:Exec
    AR_BullPup:Exec
    AR_ForwardGrip:Exec
    MedPackPaddles:Exec
    SMG_P90:Exec
    AR_DrumGun:Exec
    Consumable_Small:Exec
    Consumable_Large:Exec
    Balloon:Exec
    MountedTurret:Exec
    CreativeTool:Exec
    ExplosiveBow:Exec
    AshtonIndigo:Exec
    AshtonChicago:Exec
    MeleeDualWield:Exec
    Unarmed:Exec
   )
   Set IsGunOut(execute=[@n3.AssaultRifle, @n3.SMG_P90, @n3.TacticalShotgun], IsGunOut=1)
   Set IsPistolOut(execute=@n3.Pistol, IsPistolOut=1)
  ]}
]
variables [
 HangGliding {type=boolean}
 IsGunOut {type=number}
 IsPistolOut {type=number}
 IsSeaSaltlPA {type=boolean}
 IswarmshadeweaselPA {type=boolean}
 moveBackward {type=number}
 moveForward {type=number}
 strafeLeft float {type=number}
 strafeRight float {type=number}
]
