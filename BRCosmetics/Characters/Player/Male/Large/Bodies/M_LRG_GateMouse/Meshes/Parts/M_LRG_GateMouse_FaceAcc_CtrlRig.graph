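# Control Rig post-fix for the GateMouse large male character's face-accessory part, correcting shoulder pad and buckle bones according to gameplay state.
# One ExtraFixes chain applies weighted parent constraints to pads and chest buckle, then layers transform offsets for parachuting, directional movement, jumping, crouching and celebration emotes.
# Most offsets blend through AlphaInterp, with directional ones gated by parachute state.
# GetAlphaFromQuatDelta converts left upper-arm swing/twist into an alpha for gun-out fixes.
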ControlRigBlueprint /BRCosmetics/Characters/Player/Male/Large/Bodies/M_LRG_GateMouse/Meshes/Parts/M_LRG_GateMouse_FaceAcc_CtrlRig
graphs [
 CollapseNode_ContainedGraph {nodes [
   n1 = RigVMFunctionEntryNode()
   n2 = &RigVMFunction_MathBoolToFloat(Value=RigVMVariableNode() {variable=Pub_PadsCons}.Value)
   &RigUnit_ParentConstraint(
    AdvancedSettings=(RotationOrderForFilter=XZY)
    Child=(Type=Bone,Name="dyn_shoulderpad_02_r")
    Parents=((Item=(Type=Bone,Name="clavicle_r"),Weight=0.3093),(Item=(Type=Bone,Name="upperarm_r"),Weight=0.6907))
    ExecutePin=@n1.ExecutePin
    Weight=@n2.Result
   )
   > &RigUnit_ParentConstraint(
    AdvancedSettings=(RotationOrderForFilter=XZY)
    Child=(Type=Bone,Name="dyn_shoulderpad_02_l")
    Parents=((Item=(Type=Bone,Name="upperarm_l"),Weight=1.0))
    Weight=@n2.Result
   )
   > &RigUnit_ParentConstraint(
    AdvancedSettings=(RotationOrderForFilter=XZY)
    Child=(Type=Bone,Name="dyn_shoulderpad_02_r")
    Parents=((Item=(Type=Bone,Name="upperarm_r"),Weight=1))
    Weight=&RigVMFunction_MathBoolToFloat(Value=RigVMVariableNode() {variable=Pub_IsFE}.Value).Result
   )
   > &RigUnit_ParentConstraint(
    AdvancedSettings=(RotationOrderForFilter=XZY)
    Child=(Type=Bone,Name="base_chest_buckle")
    Parents="((Item=(Type=Bone,Name=\"spine_01\"),Weight=0.0002),(Item=(Type=Bone,Name=\"spine_04\"),Weight=0.1067),(Item=(Type=Bone,Name=\"spine_02\"),Weight=0.1214),(Item=(Type=Bone,Name=\"clavicle_r\"),Weight=0.1235),(Item=(Type=Bone,Name=\"clavicle_l\"),Weight=0.1271),(Item= ...(truncated)"
    Weight=1
    bMaintainOffset=true
   )
   RigVMFunction_MathBoolToFloat = &RigVMFunction_MathBoolToFloat(Value=RigVMVariableNode() {variable=isParachuteOpened}.Value)
   AlphaInterp = &RigVMFunction_AlphaInterp(InterpSpeedDecreasing=5, InterpSpeedIncreasing=5, Scale=1, bClampResult=False, Value=@RigVMFunction_MathBoolToFloat.Result)
   > &RigUnit_ModifyTransforms(
    ItemToModify="((Item=(Type=Bone,Name=\"dyn_shoulderpad_01_l\"),Transform=(Rotation=(X=0,Y=0.2588,Z=0,W=0.9659),Translation=(X=0,Y=0,Z=0),Scale3D=(X=1,Y=1,Z=1))),(Item=(Type=Bone,Name=\"dyn_shoulderpad_02_l\"),Transform=(Rotation=(X=0,Y=-0.342,Z=0,W=0.9397),Translation=(X=0, ...(truncated)"
    WeightMinimum=0
    Weight=@AlphaInterp.Result
   )
   > &RigUnit_ModifyTransforms(
    ItemToModify="((Item=(Type=Bone,Name=\"dyn_shoulderpad_01_r\"),Transform=(Rotation=(X=0,Y=0.2588,Z=0,W=0.9659),Translation=(X=0,Y=0,Z=0),Scale3D=(X=1,Y=1,Z=1))),(Item=(Type=Bone,Name=\"dyn_shoulderpad_02_r\"),Transform=(Rotation=(X=0,Y=-0.342,Z=0,W=0.9397),Translation=(X=0, ...(truncated)"
    WeightMinimum=0
    Weight=@AlphaInterp.Result
   )
   Multiply = &RigVMFunction_MathDoubleMul(A=@RigVMFunction_MathBoolToFloat.Result, B=RigVMVariableNode() {variable=Pub_IsLeft}.Value)
   n3 = &RigVMFunction_AlphaInterp(InterpSpeedDecreasing=5, InterpSpeedIncreasing=5, Scale=1, bClampResult=False, Value=@Multiply.Result)
   > &RigUnit_ModifyTransforms(
    ItemToModify="((Item=(Type=Bone,Name=\"dyn_shoulderpad_01_l\"),Transform=(Rotation=(X=0.0871,Y=-0.0435,Z=0.0038,W=0.9952),Translation=(X=0,Y=0,Z=0),Scale3D=(X=1,Y=1,Z=1))),(Item=(Type=Bone,Name=\"dyn_shoulderpad_02_l\"),Transform=(Rotation=(X=0,Y=0,Z=0,W=1),Translation=(X=0 ...(truncated)"
    WeightMinimum=0
    Weight=@n3.Result
   )
   n4 = &RigVMFunction_AlphaInterp(
    InterpSpeedDecreasing=5
    InterpSpeedIncreasing=5
    Scale=1
    bClampResult=False
    Value=&RigVMFunction_MathDoubleMul(A=@RigVMFunction_MathBoolToFloat.Result, B=RigVMVariableNode() {variable=Pub_IsRight}.Value).Result
   )
   > &RigUnit_ModifyTransforms(
    ItemToModify="((Item=(Type=Bone,Name=\"dyn_shoulderpad_01_r\"),Transform=(Rotation=(X=-0.0871,Y=-0.0435,Z=-0.0038,W=0.9952),Translation=(X=0,Y=0,Z=0),Scale3D=(X=1,Y=1,Z=1))),(Item=(Type=Bone,Name=\"dyn_shoulderpad_02_r\"),Transform=(Rotation=(X=0,Y=0,Z=0,W=1),Translation=(X ...(truncated)"
    WeightMinimum=0
    Weight=@n4.Result
   )
   n5 = &RigVMFunction_AlphaInterp(
    InterpSpeedDecreasing=5
    InterpSpeedIncreasing=5
    Scale=1
    bClampResult=False
    Value=&RigVMFunction_MathDoubleMul(A=@RigVMFunction_MathBoolToFloat.Result, B=RigVMVariableNode() {variable=Pub_IsForward}.Value).Result
   )
   > &RigUnit_ModifyTransforms(
    ItemToModify="((Item=(Type=Bone,Name=\"dyn_shoulderpad_01_l\"),Transform=(Rotation=(X=0.1304,Y=-0.0432,Z=0.0057,W=0.9905),Translation=(X=0,Y=0,Z=0),Scale3D=(X=1,Y=1,Z=1))),(Item=(Type=Bone,Name=\"dyn_shoulderpad_02_l\"),Transform=(Rotation=(X=0,Y=0,Z=0,W=1),Translation=(X=0 ...(truncated)"
    WeightMinimum=0
    Weight=@n5.Result
   )
   > &RigUnit_ModifyTransforms(
    ItemToModify="((Item=(Type=Bone,Name=\"dyn_shoulderpad_01_r\"),Transform=(Rotation=(X=-0.1304,Y=-0.0432,Z=-0.0057,W=0.9905),Translation=(X=0,Y=0,Z=0),Scale3D=(X=1,Y=1,Z=1))),(Item=(Type=Bone,Name=\"dyn_shoulderpad_02_r\"),Transform=(Rotation=(X=0,Y=0,Z=0,W=1),Translation=(X ...(truncated)"
    WeightMinimum=0
    Weight=@n5.Result
   )
   > GetAlphaFromQuatDelta = RigVMFunctionReferenceNode(
    Rotation Order=ZYX
    child=(Type=Bone,Name="upperarm_l")
    outputMinMax=(X=0,Y=1)
    parent=(Type=Null,Name="upperarm_l_null")
    premult=1
    rotationMinMax=(X=0,Y=-40)
    triggerAxis=Y
    twistAxis=(X=1,Y=0,Z=0)
   )
   n6 = &RigVMFunction_AlphaInterp(
    InterpSpeedDecreasing=5
    InterpSpeedIncreasing=5
    Scale=1
    bClampResult=False
    Value=&RigVMFunction_MathDoubleMul(A=@GetAlphaFromQuatDelta.alpha, B=RigVMVariableNode() {variable=Pub_IsGunOut}.Value).Result
   )
   > &RigUnit_ModifyTransforms(
    ItemToModify=((Item=(Type=Bone,Name="dyn_shoulderpad_02_l"),Transform=(Rotation=(X=0,Y=0.0436,Z=0,W=0.999),Translation=(X=3,Y=0,Z=1),Scale3D=(X=1,Y=1,Z=1))))
    WeightMinimum=0
    Weight=@n6.Result
   )
   n7 = RigVMVariableNode() {variable=Pub_IsCrouching}
   > &RigUnit_ModifyTransforms(
    ItemToModify=((Item=(Type=Bone,Name="dyn_shoulderpad_01_l"),Transform=(Rotation=(X=0.1735,Y=-0.0076,Z=-0.043,W=0.9839),Translation=(X=0,Y=0,Z=2),Scale3D=(X=1,Y=1,Z=1))))
    WeightMinimum=0
    Weight=&RigVMFunction_MathDoubleMul(A=@n6.Result, B=&RigVMFunction_MathBoolToFloat(Value=@n7.Value).Result).Result
   )
   n8 = &RigVMFunction_AlphaInterp(InterpSpeedDecreasing=5, InterpSpeedIncreasing=5, Scale=1, bClampResult=False, Value=RigVMVariableNode() {variable=Pub_IsJumping}.Value)
   > &RigUnit_ModifyTransforms(
    ItemToModify="((Item=(Type=Bone,Name=\"dyn_shoulderpad_buckle_l\"),Transform=(Rotation=(X=0,Y=-0.5736,Z=0,W=0.8192),Translation=(X=0,Y=0,Z=0),Scale3D=(X=1,Y=1,Z=1))),(Item=(Type=Bone,Name=\"dyn_shoulderpad_buckle_r\"),Transform=(Rotation=(X=0,Y=-0.5736,Z=0,W=0.8192),Transla ...(truncated)"
    WeightMinimum=0
    Weight=@n8.Result
   )
   > n9 = &RigUnit_ModifyTransforms(
    ItemToModify="((Item=(Type=Bone,Name=\"dyn_shoulderpad_02_l\"),Transform=(Rotation=(X=0,Y=0,Z=0,W=1),Translation=(X=6,Y=0,Z=0),Scale3D=(X=1,Y=1,Z=1))),(Item=(Type=Bone,Name=\"dyn_shoulderpad_01_r\"),Transform=(Rotation=(X=0,Y=-0.1736,Z=0,W=0.9848),Translation=(X=5,Y=0,Z=0), ...(truncated)"
    WeightMinimum=0
    Weight=&RigVMFunction_MathBoolToFloat(Value=RigVMVariableNode() {variable=Pub_IsCelebrationEmote}.Value).Result
   )
   RigVMFunctionReturnNode(ExecutePin=@n9.ExecutePin)
  ]}
 New Function_ContainedGraph {nodes [
   n1 = RigVMFunctionEntryNode()
   RigVMFunction_MathBoolToInteger = &RigVMFunction_MathBoolToInteger(Value=@n1.onlyTwist)
   RigVMFunction_MathQuaternionSwingTwist = &RigVMFunction_MathQuaternionSwingTwist(
    TwistAxis=@n1.twistAxis
    Input=&RigUnit_GetRelativeTransformForItem(bChildInitial=False, bParentInitial=False, Child=@n1.child, Parent=@n1.parent).RelativeTransform.Rotation
   )
   n2 = RigVMDispatchNode(
    Values=((),())
    Index=@RigVMFunction_MathBoolToInteger.Result
    Values.0=@RigVMFunction_MathQuaternionSwingTwist.Swing
    Values.1=@RigVMFunction_MathQuaternionSwingTwist.Twist
   ) {template="DISPATCH_RigVMDispatch_SelectInt32(in Index,in Values,out Result)"}
   RigVMFunction_MathQuaternionToEuler = &RigVMFunction_MathQuaternionToEuler(Value=@n2.Result, RotationOrder=@n1.Rotation Order)
   > DISPATCH_RigVMDispatch_Print = RigVMDispatchNode(ScreenColor=(R=1,G=1,B=1,A=1), ScreenDuration=1.5, Enabled=@n1.debugRotation, Value=@RigVMFunction_MathQuaternionToEuler.Result) {template="DISPATCH_RigVMDispatch_Print(in Prefix,in Value,in Enabled,in ScreenDuration,in ScreenColor)"}
   Cast = RigVMDispatchNode(Value=@n1.triggerAxis) {template="DISPATCH_RigVMDispatch_CastEnumToInt(in Value,out Result)"}
   DISPATCH_RigVMDispatch_SelectInt32 = RigVMDispatchNode(
    Values=(0,0,0)
    Index=@Cast.Result
    Values.0=@RigVMFunction_MathQuaternionToEuler.Result.X
    Values.1=@RigVMFunction_MathQuaternionToEuler.Result.Y
    Values.2=@RigVMFunction_MathQuaternionToEuler.Result.Z
   ) {template="DISPATCH_RigVMDispatch_SelectInt32(in Index,in Values,out Result)"}
   Remap = &RigVMFunction_MathDoubleRemap(
    bClamp=False
    Value=@DISPATCH_RigVMDispatch_SelectInt32.Result
    TargetMinimum=@n1.outputMinMax.X
    TargetMaximum=@n1.outputMinMax.Y
    SourceMinimum=@n1.rotationMinMax.X
    SourceMaximum=@n1.rotationMinMax.Y
   )
   Absolute = &RigVMFunction_MathDoubleAbs(Value=@Remap.Result)
   Multiply = &RigVMFunction_MathDoubleMul(
    A=&RigVMFunction_MathDoubleClamp(Value=RigVMDispatchNode(True=@Absolute.Result, Condition=@n1.absoluteValue, False=@Remap.Result) {template="DISPATCH_RigVMDispatch_If(in Condition,in True,in False,out Result)"}.Result).Result
    B=@n1.premult
   )
   > RigVMFunctionReturnNode(outputQuat=@n2.Result, outputWeight=@Multiply.Result)
  ]}
 New Function_ContainedGraph {nodes [
   RigVMFunctionEntryNode()
   RigVMFunction_MathBoolToInteger = &RigVMFunction_MathBoolToInteger(Value=RigVMVariableNode() {variable=onlyTwist}.Value)
   RigVMFunction_MathQuaternionSwingTwist = &RigVMFunction_MathQuaternionSwingTwist(
    Input=&RigUnit_GetRelativeTransformForItem(
     bChildInitial=False
     bParentInitial=False
     Child=RigVMVariableNode() {variable=child}.Value
     Parent=RigVMVariableNode() {variable=parent}.Value
    ).RelativeTransform.Rotation
    TwistAxis=RigVMVariableNode() {variable=twistAxis}.Value
   )
   n1 = RigVMDispatchNode(
    Values=((),())
    Index=@RigVMFunction_MathBoolToInteger.Result
    Values.0=@RigVMFunction_MathQuaternionSwingTwist.Swing
    Values.1=@RigVMFunction_MathQuaternionSwingTwist.Twist
   ) {template="DISPATCH_RigVMDispatch_SelectInt32(in Index,in Values,out Result)"}
   > RigVMVariableNode(Value=@n1.Result) {variable=entry_outputQuat}
   RigVMFunction_MathQuaternionToEuler = &RigVMFunction_MathQuaternionToEuler(Value=@n1.Result, RotationOrder=RigVMVariableNode() {variable=Rotation Order}.Value)
   > DISPATCH_RigVMDispatch_Print = RigVMDispatchNode(
    ScreenColor=(R=1,G=1,B=1,A=1)
    ScreenDuration=1.5
    Value=@RigVMFunction_MathQuaternionToEuler.Result
    Enabled=RigVMVariableNode() {variable=debugRotation}.Value
   ) {template="DISPATCH_RigVMDispatch_Print(in Prefix,in Value,in Enabled,in ScreenDuration,in ScreenColor)"}
   Cast = RigVMDispatchNode(Value=RigVMVariableNode() {variable=triggerAxis}.Value) {template="DISPATCH_RigVMDispatch_CastEnumToInt(in Value,out Result)"}
   DISPATCH_RigVMDispatch_SelectInt32 = RigVMDispatchNode(
    Values=(0,0,0)
    Index=@Cast.Result
    Values.0=@RigVMFunction_MathQuaternionToEuler.Result.X
    Values.1=@RigVMFunction_MathQuaternionToEuler.Result.Y
    Values.2=@RigVMFunction_MathQuaternionToEuler.Result.Z
   ) {template="DISPATCH_RigVMDispatch_SelectInt32(in Index,in Values,out Result)"}
   n2 = RigVMVariableNode() {variable=rotationMinMax}
   n3 = RigVMVariableNode() {variable=outputMinMax}
   Remap = &RigVMFunction_MathDoubleRemap(
    bClamp=False
    Value=@DISPATCH_RigVMDispatch_SelectInt32.Result
    SourceMinimum=@n2.Value.X
    SourceMaximum=@n2.Value.Y
    TargetMinimum=@n3.Value.X
    TargetMaximum=@n3.Value.Y
   )
   Absolute = &RigVMFunction_MathDoubleAbs(Value=@Remap.Result)
   n4 = RigVMVariableNode() {variable=absoluteValue}
   Multiply = &RigVMFunction_MathDoubleMul(
    A=&RigVMFunction_MathDoubleClamp(Value=RigVMDispatchNode(True=@Absolute.Result, False=@Remap.Result, Condition=@n4.Value) {template="DISPATCH_RigVMDispatch_If(in Condition,in True,in False,out Result)"}.Result).Result
    B=RigVMVariableNode() {variable=premult}.Value
   )
   > RigVMFunctionReturnNode(alpha=@Multiply.Result, rawQuat=RigVMVariableNode() {variable=entry_outputQuat}.Value)
  ]}
 RigVMFunctionLibrary {nodes [
   DeltaRotation v4 = RigVMCollapseNode(
    ExecuteContext=()
    Rotation Order=ZYX
    child=(Type=Bone,Name="None")
    outputMinMax=(X=0,Y=1)
    parent=(Type=Bone,Name="None")
    premult=1
    rotationMinMax=(X=0,Y=0)
    twistAxis=(X=1,Y=0,Z=0)
   )
   GetAlphaFromQuatDelta = RigVMCollapseNode(
    ExecuteContext=()
    Rotation Order=ZYX
    child=(Type=Bone,Name="None")
    outputMinMax=(X=0,Y=1)
    parent=(Type=Bone,Name="None")
    premult=1
    rotationMinMax=(X=0,Y=0)
    twistAxis=(X=1,Y=0,Z=0)
   )
  ]}
 RigVMModel {nodes [
   RigUnit_BeginExecution = &RigUnit_BeginExecution()
   ExtraFixes = RigVMCollapseNode(ExecutePin=@RigUnit_BeginExecution.ExecutePin)
  ]}
]
hierarchy [
 root {loc=<0 0 0>, rot=<0 0 0 1>}
 pelvis {parent=root, loc=<0 -0.7321 92.4218>, rot=<0.0127 0.707 -0.0127 -0.707>}
 spine_01 {parent=pelvis, loc=<5.6002 -0.5128 0>, rot=<0 0 0.034 -0.9994>}
 spine_02 {parent=spine_01, loc=<8.0756 0.0139 0>, rot=<0 0 0.0238 -0.9997>}
 spine_03 {parent=spine_02, loc=<7.0395 0 0>, rot=<0 0 0.0149 -0.9999>}
 spine_04 {parent=spine_03, loc=<8.1023 0 0>, rot=<0 0 -0.0132 -0.9999>}
 spine_05 {parent=spine_04, loc=<10.0141 0 0>, rot=<0 0 -0.0015 -1>}
 clavicle_l {parent=spine_05, loc=<2.5856 -0.2866 -4.3777>, rot=<0.0465 -0.7289 -0.009 -0.6829>}
 upperarm_l {parent=clavicle_l, loc=<18.2264 0.1342 -0.001>, rot=<0.0253 -0.2693 0.0013 -0.9627>}
 upperarm_twist_01_l {parent=upperarm_l, loc=<7.9796 0.0843 -0.6195>, rot=<0 0 0 -1>}
 base_shoulderpad_l {parent=upperarm_twist_01_l, loc=<-7.9796 -0.0843 0.6195>, rot=<-0.065 0.002 0.0111 -0.9978>}
 dyn_shoulderpad_01_l {parent=base_shoulderpad_l, loc=<0 0 0>, rot=<0 0 0 1>}
 dyn_shoulderpad_02_l {parent=dyn_shoulderpad_01_l, loc=<0.9088 -2.5251 10.3358>, rot=<0 0 0 1>}
 dyn_shoulderpad_buckle_l {parent=dyn_shoulderpad_02_l, loc=<0.0026 0.627 9.0194>, rot=<0.0515 -0.4509 -0.0396 -0.8902>}
 upperarm_l_null {parent=clavicle_l, type=Null, loc=<18.2264 0.1342 -0.001>, rot=<0.0253 -0.2693 0.0013 -0.9627>}
 neck_01 {parent=spine_05, loc=<10.815 1.9196 0>, rot=<0 0 -0.1864 -0.9825>}
 neck_02 {parent=neck_01, loc=<4.7347 0.3045 0>, rot=<0 0 0.0363 -0.9993>}
 head {parent=neck_02, loc=<5.7632 0.5552 0>, rot=<0 0 0.115 -0.9934>}
 clavicle_r {parent=spine_05, loc=<2.5855 -0.2866 4.3777>, rot=<0.7289 0.0465 -0.6829 0.009>}
 upperarm_r {parent=clavicle_r, loc=<-18.2264 -0.1342 0.001>, rot=<0.0253 -0.2693 0.0013 -0.9627>}
 upperarm_twist_01_r {parent=upperarm_r, loc=<-7.9796 -0.0843 0.6195>, rot=<0 0 0 -1>}
 base_shoulderpad_r {parent=upperarm_twist_01_r, loc=<7.9793 0.0843 -0.6191>, rot=<-0.0111 -0.9978 -0.065 -0.002>}
 dyn_shoulderpad_01_r {parent=base_shoulderpad_r, loc=<0 0 0>, rot=<0 0 0 1>}
 dyn_shoulderpad_02_r {parent=dyn_shoulderpad_01_r, loc=<0.9086 2.525 10.3363>, rot=<0 0 0 1>}
 dyn_shoulderpad_buckle_r {parent=dyn_shoulderpad_02_r, loc=<0.0024 -0.627 9.0196>, rot=<-0.0515 -0.4509 0.0396 -0.8902>}
 upperarm_r_null {parent=clavicle_r, type=Null, loc=<-18.2264 -0.1342 0.001>, rot=<0.0253 -0.2693 0.0013 -0.9627>}
 base_chest_buckle {parent=spine_05, loc=<2.3997 22.8295 0>, rot=<0.0401 -0.9992 0 0>}
 dyn_chest_buckle {parent=base_chest_buckle, loc=<0 0 0>, rot=<0 0 0 1>}
]
variables [
 Pub_IsBackward {type=double, default=0}
 Pub_IsCelebrationEmote {type=bool, default=False}
 Pub_IsCrouching {type=bool, default=False}
 Pub_IsFE {type=bool, default=False}
 Pub_IsForward {type=double, default=0}
 Pub_IsGunOut {type=double, default=0}
 Pub_IsJumping {type=double, default=0}
 Pub_IsLeft {type=double, default=0}
 Pub_IsRight {type=double, default=0}
 Pub_PadsCons {type=bool, default=False}
 isParachuteOpened {type=bool, default=False}
]
