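# Animation Blueprint for the GateMouse large male face accessory, layering procedural motion onto the pose copied from its attached parent.
# A Control Rig receives movement, jump, crouch, weapon, parachute and montage states from the event graph.
# Head and neck rotations are replaced by negated, scaled head deltas, updated outside emote and idle montages.
# Chained AnimDynamics springs then swing buckles and wires, using parameter-asset gravity and separate front-end and in-game sim-space settings.
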AnimBlueprint /BRCosmetics/Characters/Player/Male/Large/Bodies/M_LRG_GateMouse/Meshes/Parts/M_LRG_GateMouse_FaceAcc_AnimBP
target_skeleton=&/M_LRG_GateMouse_FaceAcc_Skeleton
graphs [
 AnimGraph {nodes [
   Output Pose(Result=Use cached pose 'CR_Cache_Output'()) {layerGroup=None}
   n1 = Control Rig!(
    Source=Copy Pose From Mesh(SourceMeshComponent=Get Skeletal_Mesh() -> (Skeletal_Mesh:Skeletal Mesh Component Object Reference)) {bUseAttachedParent=true}
    Alpha
    isParachuteOpened:bool=Get bIsParachuteOpened()
    Pub_IsLeft:double=Get strafeLeft float() -> (strafeLeft float:double)
    Pub_IsRight:double=Get strafeRight float() -> (strafeRight float:double)
    Pub_IsForward:double=Get moveForward() -> (moveForward:double)
    Pub_IsBackward:double=Get moveBackward() -> (moveBackward:double)
    Pub_IsGunOut:double=Get isGunOut() -> (isGunOut:double)
    Pub_IsCrouching:bool=Get bIsCrouching()
    Pub_IsJumping:double=Get IsJumping() -> (IsJumping:double)
    Pub_IsCelebrationEmote:bool=Get IsCelebrationEmote() -> (IsCelebrationEmote:bool)
    Pub_PadsCons:bool=Get PadsCons() -> (PadsCons:bool)
    Pub_IsFE:bool=Get IsIdleMontage() -> (IsIdleMontage:bool)
   ) -> (Pose) {controlRigAssetReference {blueprintRigClass=&/M_LRG_GateMouse_FaceAcc_CtrlRig.M_LRG_GateMouse_FaceAcc_CtrlRig_C}}
   Save cached pose 'CR_Cache_Input'(Pose=@n1) {cacheName=CR_Cache_Input}
   n2 = "Transform (Modify) Bone - Bone: head"!(ComponentPose=Local To Component(LocalPose=Use cached pose 'CR_Cache_Input'()), Alpha=1, Rotation=Get headRot() -> (headRot:Rotator)) -> (Pose) {
    rotationMode=BMM_Replace
    rotationSpace=BCS_BoneSpace
    translationSpace=BCS_BoneSpace
    alphaBoolBlend {blendInTime=0.1, blendOption=Linear, blendOutTime=0.1}
    boneToModify {boneName=head}
   }
   n3 = Get neckRot() -> (neckRot:Rotator)
   n4 = "Transform (Modify) Bone - Bone: neck_01"!(ComponentPose=@n2, Alpha=1, Rotation=@n3) -> (Pose) {
    rotationMode=BMM_Replace
    rotationSpace=BCS_BoneSpace
    translationSpace=BCS_BoneSpace
    alphaBoolBlend {blendInTime=0.1, blendOption=Linear, blendOutTime=0.1}
    boneToModify {boneName=neck_01}
   }
   n5 = "Transform (Modify) Bone - Bone: neck_02"!(ComponentPose=@n4, Alpha=1, Rotation=@n3) -> (Pose) {
    rotationMode=BMM_Replace
    rotationSpace=BCS_BoneSpace
    translationSpace=BCS_BoneSpace
    alphaBoolBlend {blendInTime=0.1, blendOption=Linear, blendOutTime=0.1}
    boneToModify {boneName=neck_02}
   }
   Save cached pose 'CR_Cache_Dynamics'(Pose=Component To Local(ComponentPose=@n5)) {cacheName=CR_Cache_Dynamics}
   n6 = Anim Dynamics - Bone: dyn_shoulderpad_buckle_l!(
    ComponentPose=Local To Component(LocalPose=Use cached pose 'CR_Cache_Dynamics'())
    Alpha=1
    GravityOverride:Vector="0,0,-980"
    SimSpaceSettings:Anim Phys Sim Space Settings Structure=(SimSpaceAngularAlpha=0,MaxAngularVelocity=10000,MaxAngularAcceleration=10000,ExternalAngularVelocity=(X=0,Y=0,Z=0))
    ExternalForce="0,0,0"
   ) -> (Pose) {
    angularDampingOverride=1
    angularSpringConstant=20000
    bAngularSpring=true
    bOverrideAngularDamping=true
    bUseGravityOverride=true
    gravityScale=0.35
    simulationSpace=RootRelative
    boundBone {boneName=dyn_shoulderpad_buckle_l}
    componentLinearAccScale=<1 1 1>
    componentLinearVelScale=<1 1 1>
    gravityOverride {z=-980}
    physicsBodyDefinitions [{
      boundBone {boneName=dyn_shoulderpad_buckle_l}
      boxExtents=<30 40 40>
      constraintSetup {angularLimitsMax=<0 0 0>, angularLimitsMin=<0 -90 0>, angularTarget=<1 0 0>}
      localJointOffset=<4 0 0>
     }]
   }
   n7 = Anim Dynamics - Bone: dyn_shoulderpad_buckle_r!(
    ComponentPose=@n6
    Alpha=1
    GravityOverride:Vector="0,0,-980"
    SimSpaceSettings:Anim Phys Sim Space Settings Structure=(SimSpaceAngularAlpha=0,MaxAngularVelocity=10000,MaxAngularAcceleration=10000,ExternalAngularVelocity=(X=0,Y=0,Z=0))
    ExternalForce="0,0,0"
   ) -> (Pose) {
    angularDampingOverride=1
    angularSpringConstant=20000
    bAngularSpring=true
    bOverrideAngularDamping=true
    bUseGravityOverride=true
    gravityScale=0.35
    simulationSpace=RootRelative
    boundBone {boneName=dyn_shoulderpad_buckle_r}
    componentLinearAccScale=<1 1 1>
    componentLinearVelScale=<1 1 1>
    gravityOverride {z=-980}
    physicsBodyDefinitions [{
      boundBone {boneName=dyn_shoulderpad_buckle_r}
      boxExtents=<30 40 40>
      constraintSetup {angularLimitsMax=<0 0 0>, angularLimitsMin=<0 -90 0>, angularTarget=<1 0 0>}
      localJointOffset=<4 0 0>
     }]
   }
   n8 = Anim Dynamics - Bone: dyn_chest_buckle!(
    ComponentPose=@n7
    Alpha=1
    GravityOverride:Vector="0,0,-980"
    SimSpaceSettings:Anim Phys Sim Space Settings Structure=(SimSpaceAngularAlpha=0,MaxAngularVelocity=10000,MaxAngularAcceleration=10000,ExternalAngularVelocity=(X=0,Y=0,Z=0))
    ExternalForce="0,0,0"
   ) -> (Pose) {
    angularDampingOverride=1
    angularSpringConstant=800
    bAngularSpring=true
    bOverrideAngularDamping=true
    bUseGravityOverride=true
    gravityScale=0.35
    simulationSpace=RootRelative
    boundBone {boneName=dyn_chest_buckle}
    componentAppliedLinearAccClamp=<10000 10000 10000>
    componentLinearAccScale=<1 1 1>
    componentLinearVelScale=<1 1 1>
    gravityOverride {z=-980}
    physicsBodyDefinitions [{
      boundBone {boneName=dyn_chest_buckle}
      boxExtents=<30 40 40>
      constraintSetup {angularLimitsMax=<0 0 60>, angularLimitsMin=<0 0 0>, angularTarget=<1 0 0>}
      localJointOffset=<4 0 0>
     }]
   }
   n9 = Anim Dynamics - Bone: dyn_wire_head_1!(
    ComponentPose=@n8
    Alpha=1
    GravityOverride:Vector="0,0,-980"
    SimSpaceSettings:Anim Phys Sim Space Settings Structure=(SimSpaceAngularAlpha=0,MaxAngularVelocity=10000,MaxAngularAcceleration=10000,ExternalAngularVelocity=(X=0,Y=0,Z=0))
    ExternalForce="0,0,0"
   ) -> (Pose) {
    angularDampingOverride=1
    angularSpringConstant=800
    bAngularSpring=true
    bOverrideAngularDamping=true
    bUseGravityOverride=true
    gravityScale=0.35
    simulationSpace=RootRelative
    boundBone {boneName=dyn_wire_head_1}
    componentAppliedLinearAccClamp=<10000 10000 10000>
    componentLinearAccScale=<1 1 1>
    componentLinearVelScale=<1 1 1>
    gravityOverride {z=-980}
    physicsBodyDefinitions [{
      boundBone {boneName=dyn_wire_head_1}
      boxExtents=<20 40 40>
      constraintSetup {angularLimitsMax=<30 15 5>, angularLimitsMin=<-30 -15 -5>, angularTarget=<1 0 0>}
      localJointOffset=<6 0 0>
     }]
   }
   n10 = Anim Dynamics - Bone: dyn_wire_head_2!(
    ComponentPose=@n9
    Alpha=1
    GravityOverride:Vector="0,0,-980"
    SimSpaceSettings:Anim Phys Sim Space Settings Structure=(SimSpaceAngularAlpha=0,MaxAngularVelocity=10000,MaxAngularAcceleration=10000,ExternalAngularVelocity=(X=0,Y=0,Z=0))
    ExternalForce="0,0,0"
   ) -> (Pose) {
    angularDampingOverride=1
    angularSpringConstant=800
    bAngularSpring=true
    bOverrideAngularDamping=true
    bUseGravityOverride=true
    gravityScale=0.35
    simulationSpace=RootRelative
    boundBone {boneName=dyn_wire_head_2}
    componentAppliedLinearAccClamp=<10000 10000 10000>
    componentLinearAccScale=<1 1 1>
    componentLinearVelScale=<1 1 1>
    gravityOverride {z=-980}
    physicsBodyDefinitions [{
      boundBone {boneName=dyn_wire_head_2}
      boxExtents=<20 40 40>
      constraintSetup {angularLimitsMax=<30 15 5>, angularLimitsMin=<-30 -15 -5>, angularTarget=<1 0 0>}
      localJointOffset=<6 0 0>
     }]
   }
   n11 = Anim Dynamics - Bone: dyn_wire_head_3!(
    ComponentPose=@n10
    Alpha=1
    GravityOverride:Vector="0,0,-980"
    SimSpaceSettings:Anim Phys Sim Space Settings Structure=(SimSpaceAngularAlpha=0,MaxAngularVelocity=10000,MaxAngularAcceleration=10000,ExternalAngularVelocity=(X=0,Y=0,Z=0))
    ExternalForce="0,0,0"
   ) -> (Pose) {
    angularDampingOverride=1
    angularSpringConstant=800
    bAngularSpring=true
    bOverrideAngularDamping=true
    bUseGravityOverride=true
    gravityScale=0.35
    simulationSpace=RootRelative
    boundBone {boneName=dyn_wire_head_3}
    componentAppliedLinearAccClamp=<10000 10000 10000>
    componentLinearAccScale=<1 1 1>
    componentLinearVelScale=<1 1 1>
    gravityOverride {z=-980}
    physicsBodyDefinitions [{
      boundBone {boneName=dyn_wire_head_3}
      boxExtents=<20 40 40>
      constraintSetup {angularLimitsMax=<30 15 5>, angularLimitsMin=<-30 -15 -5>, angularTarget=<1 0 0>}
      localJointOffset=<6 0 0>
     }]
   }
   n12 = Anim Dynamics - Bone: dyn_wire_head_4!(
    ComponentPose=@n11
    Alpha=1
    GravityOverride:Vector="0,0,-980"
    SimSpaceSettings:Anim Phys Sim Space Settings Structure=(SimSpaceAngularAlpha=0,MaxAngularVelocity=10000,MaxAngularAcceleration=10000,ExternalAngularVelocity=(X=0,Y=0,Z=0))
    ExternalForce="0,0,0"
   ) -> (Pose) {
    angularDampingOverride=1
    angularSpringConstant=800
    bAngularSpring=true
    bOverrideAngularDamping=true
    bUseGravityOverride=true
    gravityScale=0.35
    simulationSpace=RootRelative
    boundBone {boneName=dyn_wire_head_4}
    componentAppliedLinearAccClamp=<10000 10000 10000>
    componentLinearAccScale=<1 1 1>
    componentLinearVelScale=<1 1 1>
    gravityOverride {z=-980}
    physicsBodyDefinitions [{
      boundBone {boneName=dyn_wire_head_4}
      boxExtents=<20 40 40>
      constraintSetup {angularLimitsMax=<30 15 5>, angularLimitsMin=<-30 -15 -5>, angularTarget=<1 0 0>}
      localJointOffset=<6 0 0>
     }]
   }
   n13 = Anim Dynamics - Bone: dyn_wire_l!(
    ComponentPose=@n12
    Alpha=1
    GravityOverride:Vector="0,0,-980"
    SimSpaceSettings:Anim Phys Sim Space Settings Structure=(SimSpaceAngularAlpha=0,MaxAngularVelocity=10000,MaxAngularAcceleration=10000,ExternalAngularVelocity=(X=0,Y=0,Z=0))
    ExternalForce="0,0,0"
   ) -> (Pose) {
    angularDampingOverride=1
    angularSpringConstant=20000
    bAngularSpring=true
    bOverrideAngularDamping=true
    bUseGravityOverride=true
    gravityScale=0.35
    simulationSpace=RootRelative
    boundBone {boneName=dyn_wire_l}
    componentAppliedLinearAccClamp=<10000 10000 10000>
    componentLinearAccScale=<1 1 1>
    componentLinearVelScale=<1 1 1>
    gravityOverride {z=-980}
    physicsBodyDefinitions [{
      boundBone {boneName=dyn_wire_l}
      boxExtents=<30 40 40>
      constraintSetup {angularLimitsMax=<30 30 30>, angularLimitsMin=<-30 -15 -30>, angularTarget=<1 0 0>}
      localJointOffset=<4 0 0>
     }]
   }
   n14 = Anim Dynamics - Bone: dyn_wire_r!(
    ComponentPose=@n13
    Alpha=1
    GravityOverride:Vector="0,0,-980"
    SimSpaceSettings:Anim Phys Sim Space Settings Structure=(SimSpaceAngularAlpha=0,MaxAngularVelocity=10000,MaxAngularAcceleration=10000,ExternalAngularVelocity=(X=0,Y=0,Z=0))
    ExternalForce="0,0,0"
   ) -> (Pose) {
    angularDampingOverride=1
    angularSpringConstant=20000
    bAngularSpring=true
    bOverrideAngularDamping=true
    bUseGravityOverride=true
    gravityScale=0.35
    simulationSpace=RootRelative
    boundBone {boneName=dyn_wire_r}
    componentAppliedLinearAccClamp=<10000 10000 10000>
    componentLinearAccScale=<1 1 1>
    componentLinearVelScale=<1 1 1>
    gravityOverride {z=-980}
    physicsBodyDefinitions [{
      boundBone {boneName=dyn_wire_r}
      boxExtents=<30 40 40>
      constraintSetup {angularLimitsMax=<30 15 30>, angularLimitsMin=<-30 -30 -30>, angularTarget=<1 0 0>}
      localJointOffset=<4 0 0>
     }]
   }
   Save cached pose 'CR_Cache_Output'(Pose=Component To Local(ComponentPose=@n14)) {cacheName=CR_Cache_Output}
  ]}
 EventGraph {nodes [
   Event BlueprintUpdateAnimation()
   > n1 = Sequence!(execute) -> (then_0, then_1, then_2:Exec, then_3:Exec)
   > Set Skeletal_Mesh(Skeletal_Mesh:Skeletal Mesh Component Object Reference=GetOwnerSkeletalMeshComponent()) -> (Output_Get:Skeletal Mesh Component Object Reference)
   > n2 = "Gravity Override State Function (Parameter Asset) (Nativized)"(
    CharacterPartAnimInstance=Self-Reference()
    GameTimeInSeconds=GetGameTimeInSeconds()
    Parameters=&/M_LRG_GateMouse_Parameters
   )
   > Set Gravity Override(Gravity Override=@n2)
   > Set Sim Space Settings(Sim Space Settings:Anim Phys Sim Space Settings Structure=Get Sim Space Settings_BR() -> (Sim Space Settings_BR:Anim Phys Sim Space Settings Structure)) -> (Output_Get:Anim Phys Sim Space Settings Structure)
   > Branch(Condition=Get bIsInFrontEnd())
   > Set Sim Space Settings(Sim Space Settings=Get Sim Space Settings_FE() -> (Sim Space Settings_FE:Anim Phys Sim Space Settings Structure))
   n3 = Get bIsPlayingEmote()
   n4 = Get IsIdleMontage()
   Branch(execute=@n1.then_3, Condition=AND Boolean(A=NOR Boolean!(A:bool=@n3, B:bool=@n4) -> (ReturnValue:bool), B=Get LimitOn() -> (LimitOn:bool)))
   n5 = BreakRotator(InRot=GetDeltaTransformFromRefPose!(self=GetOwnerSkeletalMeshComponent(), BoneName=head, BaseName=neck_02) -> (ReturnValue_Location:Vector, ReturnValue_Rotation:Rotator, ReturnValue_Scale:Vector).ReturnValue_Rotation)
   n6 = Remap Float(Input=@n5.Yaw, OldRange=(X=10,Y=-28), NewRange=(X=0,Y=1), Clamp=true) -> (Output=~)
   n7 = Lerp(A=Get zHeadRotPerMin() -> (zHeadRotPerMin:double), B=Get zHeadRotPerMax() -> (zHeadRotPerMax:double), Alpha=@n6)
   n8 = float * float!(A=@n5, B=-1, C:double=Get xHeadRotPer() -> (xHeadRotPer:double)) -> (ReturnValue) {is=title K2Node_PromotableOperator}
   n9 = float * float(A=@n5.Pitch, B=-1, C=Get yHeadRotPer() -> (yHeadRotPer:double))
   > Set headRot(headRot:Rotator=MakeRotator(Roll=@n8, Pitch=@n9, Yaw=float * float(A=@n5.Yaw, B=-1, C=@n7))) -> (Output_Get:Rotator)
   n10 = Lerp(A=Get zHeadRotPerMin(), B=0.5, Alpha=@n6)
   > Set neckRot(neckRot:Rotator=MakeRotator(Roll=@n8, Pitch=@n9, Yaw=float * float(B=-1, C=@n10))) -> (Output_Get:Rotator)
   n11 = GetAnimInstance(self=GetOwnerSkeletalMeshComponent())
   Is Valid(exec=@n1.then_2, InputObject=@n11)
   > Set IsCelebrationEmote(IsCelebrationEmote:bool=Montage_IsPlaying(self=@n11, Montage=/Game/Animation/Game/MainPlayer/Emotes/Emote_Celebration_M)) -> (Output_Get:bool)
   n12 = PlayerMovement(execute=@n1.then_1, AnimInstance=Self-Reference(), XDotProductCoefficient=1, YDotProductCoefficient=1, ZDotProductCoefficient=1)
   n13 = BreakVector(InVec=MakeVector(X=@n12, Y=@n12.YMovement))
   > Set strafeLeft float(strafeLeft float:double="Absolute (Float)"(A="Clamp (Float)"(Value=@n13, Min=-1, Max=0))) -> (Output_Get:double)
   > Set strafeRight float(strafeRight float:double="Clamp (Float)"(Value=@n13, Min=0, Max=1)) -> (Output_Get:double)
   > Set moveForward(moveForward:double="Clamp (Float)"(Value=@n13.Y, Min=0, Max=1)) -> (Output_Get:double)
   > Set moveBackward(moveBackward:double="Absolute (Float)"(A="Clamp (Float)"(Value=@n13.Y, Min=-1, Max=0))) -> (Output_Get:double)
   n14 = GetAnimInstance(self=GetOwnerSkeletalMeshComponent())
   > Is Valid(InputObject=@n14)
   > Set IsIdleMontage(IsIdleMontage:bool=Montage_IsPlaying(self=@n14, Montage=/BRCosmetics/Animation/Game/MainPlayer/Menu/BR/CustomSkins/_s40/GateMouse_M/GateMouse_Idle_Montage)) -> (Output_Get:bool)
   > Set PadsCons(PadsCons:bool=NOT Boolean(A=Get bIsParachuteOpened())) -> (Output_Get:bool)
   Event BlueprintBeginPlay() {is=title K2Node_Event}
   > Set LimitOn(LimitOn:bool=true) -> (Output_Get:bool)
   Event OnEndFalling()
   > Set IsJumping(IsJumping:double=0) -> (Output_Get:double)
   Event OnBeginJumping()
   > Set IsJumping(IsJumping=1)
   Event OnWeaponChanged()
   > Set isGunOut(isGunOut:double=0) -> (Output_Get:double)
   > n15 = Switch on EFortWeaponCoreAnimation!(execute, Selection:EFortWeaponCoreAnimation Enum=Get WeaponCoreAnim!() -> (WeaponCoreAnim)) -> (
    Melee:Exec
    Pistol:Exec
    Shotgun:Exec
    PaperBlueprint:Exec
    Rifle:Exec
    MeleeOneHand:Exec
    MachinePistol:Exec
    RocketLauncher:Exec
    GrenadeLauncher:Exec
    GoingCommando:Exec
    AssaultRifle:Exec
    TacticalShotgun:Exec
    SniperRifle:Exec
    TrapPlacement:Exec
    ShoulderLauncher:Exec
    AbilityDecoTool:Exec
    Crossbow:Exec
    C4:Exec
    RemoteControl:Exec
    DualWield:Exec
    AR_BullPup:Exec
    AR_ForwardGrip:Exec
    MedPackPaddles:Exec
    SMG_P90:Exec
    AR_DrumGun:Exec
    Consumable_Small:Exec
    Consumable_Large:Exec
    Balloon:Exec
    MountedTurret:Exec
    CreativeTool:Exec
    ExplosiveBow:Exec
    AshtonIndigo:Exec
    AshtonChicago:Exec
    MeleeDualWield:Exec
    Unarmed:Exec
   )
   Set isGunOut(execute=[@n15.AssaultRifle, @n15.SMG_P90, @n15.TacticalShotgun], isGunOut=1)
   Reroute Node(InputPin:Wildcard) -> (OutputPin:Wildcard)
  ]}
]
variables [
 Gravity Override {type=Vector}
 headRot {type=Rotator}
 IsCelebrationEmote {type=boolean}
 isGunOut {type=number}
 IsIdleMontage {type=boolean}
 IsJumping {type=number}
 LimitOn {type=boolean}
 moveBackward {type=number}
 moveForward {type=number}
 neckRot {type=Rotator}
 PadsCons {type=boolean}
 Sim Space Settings {
  type=AnimPhysSimSpaceSettings
  default {maxAngularAcceleration=10000, maxAngularVelocity=10000, simSpaceAngularAlpha=0, externalAngularVelocity=<0 0 0>}
 }
 Sim Space Settings_BR {
  type=AnimPhysSimSpaceSettings
  default {maxAngularAcceleration=10000, maxAngularVelocity=10000, simSpaceAngularAlpha=0, externalAngularVelocity=<0 0 0>}
 }
 Sim Space Settings_FE {
  type=AnimPhysSimSpaceSettings
  default {maxAngularAcceleration=10000, maxAngularVelocity=10000, simSpaceAngularAlpha=0.5, externalAngularVelocity=<0.1 0.1 0.1>}
 }
 Skeletal_Mesh {type=SkeletalMeshComponent}
 strafeLeft float {type=number}
 strafeRight float {type=number}
 xHeadRotPer {type=number, default=0.4}
 yHeadRotPer {type=number, default=0.3}
 zHeadRotPerMax {type=number, default=1.35}
 zHeadRotPerMin {type=number}
]
